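# Sound cue for the heavy-attack swing of the Boombox sledgehammer melee weapon.
# A quality node selects between two branches, each mixing a random generic melee whoosh with the sledgehammer swing wave under randomized upward pitch.
# The second branch trims the whoosh pool from four to three, and the remaining quality slots are empty.
# Melee sound class, attenuation with a natural-sound falloff override, and player melee concurrency govern playback.
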SoundCue /Game/Sounds/Fort_Weapons/Boombox/Sledge_Hammer/Boombox_Hammer_HeavyAttack_Swing_Cue
root=QualityLevel_0
nodes [
 {id=Mixer_1, children [Random_2,Modulator_0], properties {inputVolume [0.8,1.2]}}
 {id=Mixer_2, children [Random_4,Modulator_3], properties {inputVolume [0.8,1.2]}}
 {id=Modulator_0, children [WavePlayer_7]}
 {id=Modulator_1, children [Mixer_1], properties {pitchMax=1.3, pitchMin=1}}
 {id=Modulator_2, children [Mixer_2], properties {pitchMax=1.3, pitchMin=1}}
 {id=Modulator_3, children [WavePlayer_8]}
 {id=QualityLevel_0, children [Modulator_1,Modulator_2,null,null]}
 {id=Random_2, children [WavePlayer_9, WavePlayer_10, WavePlayer_11, WavePlayer_12], properties {weights [1,1,1,1]}}
 {id=Random_4, children [WavePlayer_18, WavePlayer_19, WavePlayer_21], properties {weights [1,1,1]}}
 {id=WavePlayer_10, properties {soundWaveAssetPtr=/Game/Sounds/Fort_Weapons/Melee/rightmouse_generic_whoosh_03}}
 {id=WavePlayer_11, properties {soundWaveAssetPtr=/Game/Sounds/Fort_Weapons/Melee/rightmouse_generic_whoosh_02}}
 {id=WavePlayer_12, properties {soundWaveAssetPtr=/Game/Sounds/Fort_Weapons/Melee/rightmouse_generic_whoosh_04}}
 {id=WavePlayer_18, properties {soundWaveAssetPtr=/Game/Sounds/Fort_Weapons/Melee/rightmouse_generic_whoosh_01}}
 {id=WavePlayer_19, properties {soundWaveAssetPtr=&rightmouse_generic_whoosh_03}}
 {id=WavePlayer_21, properties {soundWaveAssetPtr=&rightmouse_generic_whoosh_04}}
 {id=WavePlayer_7, properties {soundWaveAssetPtr=&/BoomBox_Sledgehammer_Swing_02}}
 {id=WavePlayer_8, properties {soundWaveAssetPtr=&BoomBox_Sledgehammer_Swing_02}}
 {id=WavePlayer_9, properties {soundWaveAssetPtr=&rightmouse_generic_whoosh_01}}
]
properties {
 attenuationSettings=&Fort_Weapon_Melee_Sml_Att
 maxDistance=2800
 soundClassObject=&Fort_WeaponMelee_PSM
 volumeMultiplier=1
 attenuationOverrides {distanceAlgorithm=NaturalSound, falloffDistance=2000, attenuationShapeExtents {x=500}}
 concurrencySet [&Player_Melee_Weapon_Conc]
}
